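# Physics asset for the SK_Thermal_Scope_Rifle skeletal mesh, a thermal scope assault rifle in the Figment content folder.
# It holds a single capsule body on the Handle bone, offset and rotated to lie roughly along the bone's X axis.
# It defines no constraints, so the rifle collides and simulates as one rigid shape with no articulated parts.
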PhysicsAsset /Figment_Assault_Surgical_Thermal_Athena/Meshes/SK_Thermal_Scope_Rifle_PhysicsAsset
bodies {Handle [{shapeType=Capsule, length=47.3063, radius=21.4723, center=<21.7492 -0.3358 3.8392>, rotation=<rot 0 89.363 -87.3647>}]}
constraints []
