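# First-person gunfire SoundCue for the Figment thermal rifle, built from M4A1 fire waves.
# A Mixer blends two layers, a main fire shot and an extra fire layer, each randomly picking one of four waves through a Modulator for slight pitch and volume variation.
# It uses first-person weapon attenuation and sound class, with a gunfire concurrency set plus owner-limited overrides that stop the oldest sound.
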SoundCue /Figment_Assault_Surgical_Thermal_Athena/Sounds/ThermalRifle_Figment_Fire_1P_Cue
root=Mixer_0
nodes [
 {id=Mixer_0, children [Modulator_1,Modulator_0], properties {inputVolume [0.7,0.9]}}
 {id=Modulator_0, children [Random_1], properties {pitchMax=1.1, pitchMin=1, volumeMax=1.1, volumeMin=0.9}}
 {id=Modulator_1, children [Random_0], properties {pitchMax=1.1, pitchMin=1, volumeMax=1.1, volumeMin=0.9}}
 {id=Random_0, children [WavePlayer_0,WavePlayer_1,WavePlayer_6,WavePlayer_7], properties {weights [1,1,1,1]}}
 {id=Random_1, children [WavePlayer_8, WavePlayer_9, WavePlayer_10, WavePlayer_11], properties {weights [1,1,1,1]}}
 {id=WavePlayer_0, properties {soundWaveAssetPtr=&/Weapon_M4A1_FireC01}}
 {id=WavePlayer_1, properties {soundWaveAssetPtr=&/Weapon_M4A1_FireC02}}
 {id=WavePlayer_10, properties {soundWaveAssetPtr=&/Weapon_M4A1_Fire_Layer03}}
 {id=WavePlayer_11, properties {soundWaveAssetPtr=&/Weapon_M4A1_Fire_Layer04}}
 {id=WavePlayer_6, properties {soundWaveAssetPtr=&/Weapon_M4A1_FireC03}}
 {id=WavePlayer_7, properties {soundWaveAssetPtr=&/Weapon_M4A1_FireC04}}
 {id=WavePlayer_8, properties {soundWaveAssetPtr=&/Weapon_M4A1_Fire_Layer01}}
 {id=WavePlayer_9, properties {soundWaveAssetPtr=&/Weapon_M4A1_Fire_Layer02}}
]
properties {
 attenuationSettings=&Fort_Weapon_Projectile_FirstPerson_Athena_Att
 bDebug=true
 maxDistance=2097152
 soundClassObject=&Fort_WeaponRanged_Fire_1P_PSM
 volumeMultiplier=0.96
 concurrencyOverrides {bLimitToOwner=true, maxCount=3, resolutionRule=StopOldest}
 concurrencySet [&Weapon_Gunfire_FirstPerson_Conc]
}
