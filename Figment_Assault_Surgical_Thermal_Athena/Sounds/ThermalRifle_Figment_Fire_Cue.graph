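# Fire sound cue for the Figment thermal rifle, built from randomized, pitch- and volume-modulated M4A1 samples.
# A close branch mixes main and layer shots under scoped attenuation and the 3P fire sound class.
# A second branch mixes medium and distant shots, each with its own mono attenuation, under the distant fire class.
# Owner-limited concurrency stops the oldest instance, alongside local and global gunfire concurrency sets.
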SoundCue /Figment_Assault_Surgical_Thermal_Athena/Sounds/ThermalRifle_Figment_Fire_Cue
root=Mixer_1
nodes [
 {
  id=Attenuation_118
  children [Mixer_5]
  properties {attenuationSettings=/Game/Sounds/Attenuation/Fort_Weapon_Projectile_Close_Athena_Att_Scoped}
 }
 {
  id=Attenuation_119
  children [Modulator_0]
  properties {attenuationSettings=&Fort_Weapon_Projectile_Medium_Mono_Athena_Att}
 }
 {
  id=Attenuation_120
  children [Modulator_1]
  properties {attenuationSettings=&Fort_Weapon_Projectile_Distant_Mono_Athena_Att}
 }
 {id=Mixer_1, children [SoundClass_77, SoundClass_76], properties {inputVolume [1,0.7]}}
 {id=Mixer_4, children [Attenuation_119, Attenuation_120], properties {inputVolume [0.8,0.6]}}
 {id=Mixer_5, children [Modulator_5,Modulator_4], properties {inputVolume [0.7,0.9]}}
 {id=Modulator_0, children [Random_8], properties {pitchMax=1.2, pitchMin=1.1, volumeMax=2, volumeMin=2}}
 {id=Modulator_1, children [Random_9], properties {pitchMax=1.2, pitchMin=1.1, volumeMax=2, volumeMin=2}}
 {id=Modulator_4, children [Random_2], properties {pitchMax=1.1, pitchMin=1, volumeMax=1.1, volumeMin=0.9}}
 {id=Modulator_5, children [Random_1], properties {pitchMax=1.1, pitchMin=1, volumeMax=1.1, volumeMin=0.9}}
 {id=Random_1, children [WavePlayer_0, WavePlayer_1, WavePlayer_6, WavePlayer_12], properties {weights [1,1,1,1]}}
 {id=Random_2, children [WavePlayer_13, WavePlayer_14, WavePlayer_16, WavePlayer_26], properties {weights [1,1,1,1]}}
 {id=Random_8, children [WavePlayer_7,WavePlayer_8,WavePlayer_9], properties {weights [1,1,1]}}
 {id=Random_9, children [WavePlayer_10, WavePlayer_11, WavePlayer_15], properties {weights [1,1,1]}}
 {id=SoundClass_76, children [Mixer_4], properties {soundClassOverride=&Fort_WeaponRanged_Fire_Distant}}
 {
  id=SoundClass_77
  children [Attenuation_118]
  properties {soundClassOverride=&Fort_WeaponRanged_Fire_3P_PSM}
 }
 {id=WavePlayer_0, properties {soundWaveAssetPtr=&/Weapon_M4A1_FireC01}}
 {id=WavePlayer_1, properties {soundWaveAssetPtr=&/Weapon_M4A1_FireC02}}
 {id=WavePlayer_10, properties {soundWaveAssetPtr=&/Weapon_M4A1_Fire_Distant03}}
 {id=WavePlayer_11, properties {soundWaveAssetPtr=&/Weapon_M4A1_Fire_Distant02}}
 {id=WavePlayer_12, properties {soundWaveAssetPtr=&/Weapon_M4A1_FireC04}}
 {id=WavePlayer_13, properties {soundWaveAssetPtr=&/Weapon_M4A1_Fire_Layer01}}
 {id=WavePlayer_14, properties {soundWaveAssetPtr=&/Weapon_M4A1_Fire_Layer02}}
 {id=WavePlayer_15, properties {soundWaveAssetPtr=&/Weapon_M4A1_Fire_Distant01}}
 {id=WavePlayer_16, properties {soundWaveAssetPtr=&/Weapon_M4A1_Fire_Layer03}}
 {id=WavePlayer_26, properties {soundWaveAssetPtr=&/Weapon_M4A1_Fire_Layer04}}
 {id=WavePlayer_6, properties {soundWaveAssetPtr=&/Weapon_M4A1_FireC03}}
 {id=WavePlayer_7, properties {soundWaveAssetPtr=&/Weapon_M4A1_Fire_Medium_01}}
 {id=WavePlayer_8, properties {soundWaveAssetPtr=&/Weapon_M4A1_Fire_Medium_02}}
 {id=WavePlayer_9, properties {soundWaveAssetPtr=&/Weapon_M4A1_Fire_Medium_03}}
]
properties {
 bDebug=true
 maxDistance=27500
 soundClassObject=&Fort_WeaponRanged_Fire_3P_PSM
 volumeMultiplier=0.72
 concurrencyOverrides {bLimitToOwner=true, maxCount=3, resolutionRule=StopOldest}
 concurrencySet [&Weapon_Gunfire_Local_Conc, &Weapon_Gunfire_Global_Conc]
}
